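# Niagara swing trail for the CymbalBar pickaxe, drawing animation-driven ribbons between pairs of bones or sockets.
# Stripes, Stars and Mix emitters share anim-trail spawn, initialize and update modules.
# User.SwingSectionIndex drives each emitter's lifetime switch, zeroing the others so each swing section shows one variant: Stripes on 0 and 2, Stars on 1, Mix on 3.
# Color takes a time-of-day multiplier and brightness boost, while alpha fades by curve and System.TrailFade.
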NiagaraSystem /BRCosmetics/Weapons/FORT_Melee/Pickaxe_CymbalBar/FX/NS_Pickaxe_CymbalBar_Trail
dependencies {
 usedDataInterfaces [&NiagaraDataInterfaceCurve, &NiagaraDataInterfaceSocketReader, &NiagaraDataInterfaceColorCurve]
 usedDynamicInputs [&ColorFromCurve]
 usedModules [
  &self:SystemSpawnScript.SystemScriptSource.SystemScriptGraph.NiagaraNodeAssignment_1.SetVariables_#1
  &SystemState
  &self:SystemSpawnScript.SystemScriptSource.SystemScriptGraph.NiagaraNodeAssignment_0.SetVariables_#2
  &self:SystemSpawnScript.SystemScriptSource.SystemScriptGraph.NiagaraNodeAssignment_3.SetVariables_#3
  &self:SystemSpawnScript.SystemScriptSource.SystemScriptGraph.NiagaraNodeAssignment_2.SetVariables_#4
  &self:SystemSpawnScript.SystemScriptSource.SystemScriptGraph.NiagaraNodeAssignment_4.SetVariables_#5
  &self:SystemSpawnScript.SystemScriptSource.SystemScriptGraph.NiagaraNodeAssignment_6.SetVariables_#6
 ]
}
emitters {
 Mix {
  emitterUpdateScript [
   &EmitterState(Life Cycle Mode=::NewEnumerator1, Loop Behavior=::NewEnumerator1, Loop Delay=0.05, Loop Duration=0.25, UseLoopDelay=false)
   &FN_SpawnAnimTrail(Spawn Interval=0)
  ]
  particleSpawnScript [
   &FN_InitializeAnimTrailBoneIndecies(First Bone or Socket Index=5, Second Bone or Socket Index=4) {module=FN_InitializeAnimTrailBoneIndecies001}
   &FN_InitializeAnimTrail(
    Color=$System.TODColorMultiplier
    Lifetime=&MergeHelper_FloatSwitch(0=0, 1=0, 2=0, 3=0.16, Index=$User.SwingSectionIndex, Num=::NewEnumerator2)
   )
  ]
  particleUpdateScript [
   &FN_UpdateAnimTrail() {module=FN_UpdateAnimTrail001}
   &ScaleColor(
    Scale Alpha=&Multiply_Float(
     A=&FloatFromCurve(
      FloatCurve=DataInterface() {propertyValues {Curve {
         defaultValue=0.5
         keys [{interpMode=RCIM_Cubic, tangentMode=RCTM_User, time=0, value=1}, {interpMode=RCIM_Cubic, tangentMode=RCTM_User, time=1, value=0}]
        }}}
      Scale Curve=2
     )
     B=$System.TrailFade
    )
    Scale RGB=&MakeVectorFromLinearColorRGB(Color=<rgba 5 5 5 0>)
    ScaleA=true
    ScaleRGB=true
   )
  ]
  rendererClasses [&NiagaraRibbonRendererProperties]
  renderers [{rendererClass=&NiagaraRibbonRendererProperties}]
 }
 Stars {
  emitterUpdateScript [
   &EmitterState(Life Cycle Mode=::NewEnumerator1, Loop Behavior=::NewEnumerator1, Loop Delay=0.05, Loop Duration=0.25, UseLoopDelay=false)
   &FN_SpawnAnimTrail(Spawn Interval=0)
  ]
  particleSpawnScript [
   &FN_InitializeAnimTrailBoneIndecies(
    First Bone or Socket Index:NiagaraInt32=&MergeHelper_IntSwitch(0=2, 1=2, 2=2, 3=4, 4=?, 5=?, Index=$User.SwingSectionIndex, Num=::NewEnumerator2)
    Second Bone or Socket Index:NiagaraInt32=&MergeHelper_IntSwitch(0=3, 1=3, 2=3, 3=5, 4=?, 5=?, Index=$User.SwingSectionIndex, Num=::NewEnumerator2)
   ) {module=FN_InitializeAnimTrailBoneIndecies001}
   &FN_InitializeAnimTrail(
    Color=$System.TODColorMultiplier
    Lifetime=&MergeHelper_FloatSwitch(0=0, 1=0.16, 2=0, 3=0, Index=$User.SwingSectionIndex, Num=::NewEnumerator2)
   )
  ]
  particleUpdateScript [
   &FN_UpdateAnimTrail() {module=FN_UpdateAnimTrail001}
   &ScaleColor(
    Scale Alpha=&Multiply_Float(
     A=&FloatFromCurve(
      FloatCurve=DataInterface() {propertyValues {Curve {
         defaultValue=0.5
         keys [{interpMode=RCIM_Cubic, tangentMode=RCTM_User, time=0, value=1}, {interpMode=RCIM_Cubic, tangentMode=RCTM_User, time=1, value=0}]
        }}}
      Scale Curve=2
     )
     B=$System.TrailFade
    )
    Scale RGB=&MakeVectorFromLinearColorRGB(Color=<rgba 5 5 5 0>)
    ScaleA=true
    ScaleRGB=true
   )
  ]
  rendererClasses [&NiagaraRibbonRendererProperties]
  renderers [{rendererClass=&NiagaraRibbonRendererProperties}]
 }
 Stripes {
  emitterUpdateScript [
   &EmitterState(Life Cycle Mode=::NewEnumerator1, Loop Behavior=::NewEnumerator1, Loop Delay=0.05, Loop Duration=0.25, UseLoopDelay=false)
   &FN_SpawnAnimTrail(Spawn Interval=0)
  ]
  particleSpawnScript [
   &FN_InitializeAnimTrailBoneIndecies(First Bone or Socket Index=2, Second Bone or Socket Index=3) {module=FN_InitializeAnimTrailBoneIndecies001}
   &FN_InitializeAnimTrail(
    Color=$System.TODColorMultiplier
    Lifetime=&MergeHelper_FloatSwitch(0=0.16, 1=0, 2=0.16, 3=0, Index=$User.SwingSectionIndex, Num=::NewEnumerator2)
   )
  ]
  particleUpdateScript [
   &FN_UpdateAnimTrail() {module=FN_UpdateAnimTrail001}
   &ScaleColor(
    Scale Alpha=&Multiply_Float(
     A=&FloatFromCurve(
      FloatCurve=DataInterface() {propertyValues {Curve {
         defaultValue=0.5
         keys [{interpMode=RCIM_Cubic, tangentMode=RCTM_User, time=0, value=1}, {interpMode=RCIM_Cubic, tangentMode=RCTM_User, time=1, value=0}]
        }}}
      Scale Curve=2
     )
     B=$System.TrailFade
    )
    Scale RGB=&MakeVectorFromLinearColorRGB(Color=<rgba 5 5 5 0>)
    ScaleA=true
    ScaleRGB=true
   )
  ]
  rendererClasses [&NiagaraRibbonRendererProperties]
  renderers [{rendererClass=&NiagaraRibbonRendererProperties}]
 }
}
summary {
 systemName=NS_Pickaxe_CymbalBar_Trail
 emitters [
  {emitterName=Stripes, rendererClasses [&NiagaraRibbonRendererProperties]}
  {emitterName=Stars, rendererClasses [&NiagaraRibbonRendererProperties]}
  {emitterName=Mix, rendererClasses [&NiagaraRibbonRendererProperties]}
 ]
 userVariables [
  User.bIsFrontEnd {type=NiagaraBool}
  User.bIsFrontEndPreview {type=NiagaraBool}
  User.RibbonIDOffset {type=NiagaraInt32}
  User.SwingSectionIndex {type=NiagaraInt32}
 ]
}
